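# Physics asset for the large female PressureGhoul player body cosmetic, set up for secondary motion on its collar, jacket front and D-rings.
# Capsules, plus a few boxes, cover the torso, head, limbs and dangling parts.
# Tiny BASE_ anchor bodies join pelvis or spine through free angular constraints.
# Each dangling body hangs from its anchor with angular limits, mostly narrow on twist, some tuned differently per side.
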PhysicsAsset /BRCosmetics/Characters/Player/Female/Large/Bodies/F_LRG_PressureGhoul/Meshes/F_LRG_PressureGhoul_Physics
bodies {
 BASE_L_dring_joint [{shapeType=Capsule, length=0.505, radius=0.505, shapeName=BASE_L_dring_joint_capsule, rotation=<rot 0 -90 0>}]
 BASE_R_dring_joint [{shapeType=Capsule, length=0.505, radius=0.505, shapeName=BASE_R_dring_joint_capsule, rotation=<rot 0 -90 0>}]
 BASE_dyn_collar_l_01 [{shapeType=Capsule, length=0.505, radius=0.505, shapeName=BASE_dyn_collar_l_01_capsule, rotation=<rot 0 -90 0>}]
 BASE_dyn_collar_r_01 [{shapeType=Capsule, length=0.505, radius=0.505, shapeName=BASE_dyn_collar_r_01_capsule, rotation=<rot 0 -90 0>}]
 BASE_dyn_jacket_fr_l_01 [{shapeType=Capsule, length=0.505, radius=0.505, shapeName=BASE_dyn_jacket_fr_l_01_capsule, rotation=<rot 0 -90 0>}]
 BASE_dyn_jacket_fr_r_01 [{shapeType=Capsule, length=0.505, radius=0.505, shapeName=BASE_dyn_jacket_fr_r_01_capsule, rotation=<rot 0 -90 0>}]
 L_dring_joint [{
   shapeType=Capsule
   length=3.2405
   radius=2.1137
   shapeName=L_dring_joint_capsule
   center=<2.0793 -0.074 1.4335>
   rotation=<rot 0 1.5131 -87.6791>
  }]
 R_dring_joint [
  {
   shapeType=Capsule
   length=3.2405
   radius=2.1137
   shapeName=R_dring_joint_capsule
   center=<-1.5486 0.0897 -1.3937>
   rotation=<rot 0 1.5131 -87.6791>
  }
  {
   shapeType=Capsule
   length=3.2405
   radius=2.1137
   shapeName=R_dring_joint_capsule
   center=<-3.7785 -0.2334 -1.3937>
   rotation=<rot 0 1.5131 -87.6791>
  }
 ]
 calf_l [{shapeType=Capsule, length=30.5484, radius=10.8328, center=<-23.776 -5.0706 -0.0591>, rotation=<rot -0.0077 -94.5721 90.108>}]
 calf_r [{shapeType=Capsule, length=30.5484, radius=10.8328, center=<23.776 5.0706 0.0591>, rotation=<rot -0.0077 -94.5721 -89.892>}]
 clavicle_l [{shapeType=Capsule, length=20.9787, radius=5.8924, center=<3.8306 -0.4087 0.4242>, rotation=<rot 0 -82.7612 -89.1886>}]
 clavicle_r [{
   shapeType=Capsule
   length=3.1216
   radius=4.301
   shapeName=clavicle_r_capsule
   center=<0.4841 5.6185 -1.1614>
   rotation=<rot -15.7494 -76.4296 -91.599>
  }]
 dyn_big_collar_l_01 [
  {shapeType=Capsule, length=10, radius=10.5, shapeName=dyn_big_collar_l_01_capsule, center=<0.855 0 9.3854>, rotation=<rot 90 180 179.9998>}
  {
   shapeType=Box
   extentX=11.7784
   extentY=11.2998
   extentZ=2.6916
   shapeName=dyn_big_collar_l_01_box
   center=<2.7434 -1.2911 18.5364>
   rotation=<rot 3.3022 52.1895 -3.5122>
  }
 ]
 dyn_big_collar_r_01 [
  {
   shapeType=Capsule
   length=10
   radius=12.5
   shapeName=dyn_big_collar_r_01_capsule
   center=<2.141 0.8397 11.048>
   rotation=<rot 85 179.9986 179.9991>
  }
  {
   shapeType=Box
   extentX=10.7363
   extentY=13.2175
   extentZ=2.4572
   shapeName=dyn_big_collar_l_01_box
   center=<1.1758 1.1548 23.0251>
   rotation=<rot 3.1268 -48.7851 179.5879>
  }
 ]
 dyn_jacket_fr_l_01 [{
   shapeType=Capsule
   length=8.2444
   radius=6.3569
   shapeName=dyn_jacket_fr_l_01_capsule
   center=<6.942 -1.7811 4.2267>
   rotation=<rot 0 161.1119 -78.8262>
  }]
 dyn_jacket_fr_r_01 [{
   shapeType=Capsule
   length=7.5438
   radius=5.556
   shapeName=dyn_jacket_fr_r_01_capsule
   center=<7.5576 1.3429 3.5362>
   rotation=<rot 0 14.6097 -88.5394>
  }]
 foot_l [{shapeType=Capsule, length=13.1659, radius=7.799, center=<-4.8725 5.8839 -0.084>, rotation=<rot -0.6721 179.2464 -91.4399>}]
 foot_r [{shapeType=Capsule, length=8.4159, radius=8.7988, center=<5.1265 -5.7041 0.7332>, rotation=<rot 0.6234 1.1655 -88.5384>}]
 hand_l [{shapeType=Capsule, length=8.5172, radius=6.7354, center=<7.2196 -2.4685 0.0559>, rotation=<rot 0 88.1479 -58.3382>}]
 hand_r [{shapeType=Capsule, length=8.3026, radius=7.2545, center=<-6.9212 2.25 0.0794>, rotation=<rot 0 96.1864 -52.3053>}]
 head [{shapeType=Capsule, length=2.0719, radius=17.5, center=<13.9496 -0.0001 0>, rotation=<rot 0 -90 0>}]
 lowerarm_l [{shapeType=Capsule, length=35.0482, radius=5.8186, center=<17.5656 -2.5798 -1.2337>, rotation=<rot 2.7686 -102.2653 -86.1165>}]
 lowerarm_r [{shapeType=Capsule, length=26.7452, radius=5.8107, center=<-18.9951 1.8597 2.7659>, rotation=<rot 0 -94.0739 -84.6099>}]
 pelvis [
  {shapeType=Capsule, length=13.208, radius=17.8423, center=<-0.9839 -2.4871 -0.4806>, rotation=<rot -4.9681 -82.8847 -0.5913>}
  {shapeType=Box, extentX=7.5, extentY=7.5, extentZ=7.5}
 ]
 spine_01 [{shapeType=Capsule, length=19.9716, radius=16.8828, center=<0 -0.801 0>, rotation=<rot 0 -90 0>}]
 spine_03 [{shapeType=Capsule, length=5.001, radius=13.9514, center=<4.7629 -1.9942 -0.0437>, rotation=<rot -89.4736 179.9706 -180>}]
 spine_04 [{shapeType=Capsule, length=7.2641, radius=15.6748, center=<0 -1.0312 0>}]
 spine_05 [{shapeType=Capsule, length=20.0943, radius=13.5568}]
 thigh_l [{shapeType=Capsule, length=3.5993, radius=6.9083, shapeName=thigh_l_capsule, center=<-10.077 0 -15.7201>, rotation=<rot 0 -90.0001 89.1972>}]
 thigh_twist_01_l [{shapeType=Capsule, length=30.3336, radius=11.7673, center=<4.5806 -4.0816 -2.8555>, rotation=<rot 0.2369 -88.585 91.0923>}]
 thigh_twist_01_r [{shapeType=Capsule, length=30.3336, radius=11.7673, center=<-5.9863 4.8858 1.9782>, rotation=<rot 0.1146 -82.2321 91.1118>}]
 upperarm_r [
  {shapeType=Capsule, length=4.6654, radius=6.7035, center=<-6.3293 1.3058 0.0149>, rotation=<rot -14.813 -92.5958 -92.3309>}
  {
   shapeType=Capsule
   length=8.3602
   radius=8.323
   shapeName=upperarm_r_capsule
   center=<-4.1677 28.7032 3.7427>
   rotation=<rot -12.2526 -91.9152 -103.1472>
  }
 ]
 upperarm_twist_01_l [
  {shapeType=Capsule, length=20.9787, radius=5.8924, center=<3.8306 -0.4087 0.4242>, rotation=<rot 0 -82.7612 -89.1886>}
  {
   shapeType=Capsule
   length=6.6946
   radius=7.0786
   shapeName=upperarm_twist_01_l_capsule
   center=<0.2739 -0.8708 0.3259>
   rotation=<rot 0 -82.7612 -89.1886>
  }
 ]
}
constraints [
 {bone1Name=spine_03, bone2Name=pelvis, swing1LimitDegrees=45, swing2LimitDegrees=45, twistLimitDegrees=45}
 {bone1Name=spine_04, bone2Name=spine_03, swing1LimitDegrees=45, swing2LimitDegrees=45, twistLimitDegrees=45}
 {bone1Name=spine_05, bone2Name=spine_04, swing1LimitDegrees=45, swing2LimitDegrees=45, twistLimitDegrees=45}
 {bone1Name=clavicle_l, bone2Name=spine_05, swing1LimitDegrees=45, swing2LimitDegrees=45, twistLimitDegrees=45}
 {bone1Name=lowerarm_l, bone2Name=clavicle_l, swing1LimitDegrees=45, swing2LimitDegrees=45, twistLimitDegrees=45}
 {bone1Name=hand_l, bone2Name=lowerarm_l, swing1LimitDegrees=45, swing2LimitDegrees=45, twistLimitDegrees=45}
 {bone1Name=upperarm_twist_01_l, bone2Name=clavicle_l, swing1LimitDegrees=45, swing2LimitDegrees=45, twistLimitDegrees=45}
 {bone1Name=hand_r, bone2Name=lowerarm_r, swing1LimitDegrees=45, swing2LimitDegrees=45, twistLimitDegrees=45}
 {bone1Name=calf_l, bone2Name=pelvis, swing1LimitDegrees=45, swing2LimitDegrees=45, twistLimitDegrees=45}
 {bone1Name=foot_l, bone2Name=calf_l, swing1LimitDegrees=45, swing2LimitDegrees=45, twistLimitDegrees=45}
 {bone1Name=thigh_twist_01_l, bone2Name=pelvis, swing1LimitDegrees=45, swing2LimitDegrees=45, twistLimitDegrees=45}
 {bone1Name=calf_r, bone2Name=pelvis, swing1LimitDegrees=45, swing2LimitDegrees=45, twistLimitDegrees=45}
 {bone1Name=foot_r, bone2Name=calf_r, swing1LimitDegrees=45, swing2LimitDegrees=45, twistLimitDegrees=45}
 {bone1Name=thigh_twist_01_r, bone2Name=pelvis, swing1LimitDegrees=45, swing2LimitDegrees=45, twistLimitDegrees=45}
 {
  bone1Name=BASE_L_dring_joint
  bone2Name=pelvis
  swing1LimitDegrees=45
  swing1Motion=Free
  swing2LimitDegrees=45
  swing2Motion=Free
  twistLimitDegrees=45
  twistMotion=Free
 }
 {bone1Name=L_dring_joint, bone2Name=BASE_L_dring_joint, swing1LimitDegrees=3, swing2LimitDegrees=60, twistLimitDegrees=3}
 {
  bone1Name=BASE_R_dring_joint
  bone2Name=pelvis
  swing1LimitDegrees=45
  swing1Motion=Free
  swing2LimitDegrees=45
  swing2Motion=Free
  twistLimitDegrees=45
  twistMotion=Free
 }
 {bone1Name=R_dring_joint, bone2Name=BASE_R_dring_joint, swing1LimitDegrees=3, swing2LimitDegrees=80, twistLimitDegrees=3}
 {
  bone1Name=spine_01
  bone2Name=pelvis
  swing1LimitDegrees=45
  swing1Motion=Free
  swing2LimitDegrees=45
  swing2Motion=Free
  twistLimitDegrees=45
  twistMotion=Free
 }
 {bone1Name=spine_03, bone2Name=spine_01, swing1LimitDegrees=45, swing2LimitDegrees=45, twistLimitDegrees=45}
 {
  bone1Name=BASE_dyn_jacket_fr_l_01
  bone2Name=spine_04
  swing1LimitDegrees=45
  swing1Motion=Free
  swing2LimitDegrees=45
  swing2Motion=Free
  twistLimitDegrees=45
  twistMotion=Free
 }
 {bone1Name=dyn_jacket_fr_l_01, bone2Name=BASE_dyn_jacket_fr_l_01, swing1LimitDegrees=20, swing2LimitDegrees=20, twistLimitDegrees=5}
 {
  bone1Name=BASE_dyn_jacket_fr_r_01
  bone2Name=spine_04
  swing1LimitDegrees=45
  swing1Motion=Free
  swing2LimitDegrees=45
  swing2Motion=Free
  twistLimitDegrees=45
  twistMotion=Free
 }
 {bone1Name=dyn_jacket_fr_r_01, bone2Name=BASE_dyn_jacket_fr_r_01, swing1LimitDegrees=10, swing2LimitDegrees=20, twistLimitDegrees=5}
 {
  bone1Name=BASE_dyn_collar_l_01
  bone2Name=spine_05
  swing1LimitDegrees=45
  swing1Motion=Free
  swing2LimitDegrees=45
  swing2Motion=Free
  twistLimitDegrees=45
  twistMotion=Free
 }
 {bone1Name=dyn_big_collar_l_01, bone2Name=BASE_dyn_collar_l_01, swing1LimitDegrees=5, swing2LimitDegrees=80, twistLimitDegrees=5}
 {
  bone1Name=BASE_dyn_collar_r_01
  bone2Name=spine_05
  swing1LimitDegrees=45
  swing1Motion=Free
  swing2LimitDegrees=45
  swing2Motion=Free
  twistLimitDegrees=45
  twistMotion=Free
 }
 {bone1Name=dyn_big_collar_r_01, bone2Name=BASE_dyn_collar_r_01, swing1LimitDegrees=10, swing2LimitDegrees=60, twistLimitDegrees=30}
 {bone1Name=upperarm_r, bone2Name=clavicle_r, swing1LimitDegrees=45, swing2LimitDegrees=45, twistLimitDegrees=45}
 {bone1Name=lowerarm_r, bone2Name=upperarm_r, swing1LimitDegrees=45, swing2LimitDegrees=45, twistLimitDegrees=45}
 {
  bone1Name=head
  bone2Name=spine_05
  swing1LimitDegrees=45
  swing1Motion=Free
  swing2LimitDegrees=45
  swing2Motion=Free
  twistLimitDegrees=45
  twistMotion=Free
 }
 {
  bone1Name=thigh_l
  bone2Name=pelvis
  swing1LimitDegrees=45
  swing1Motion=Free
  swing2LimitDegrees=45
  swing2Motion=Free
  twistLimitDegrees=45
  twistMotion=Free
 }
 {bone1Name=calf_l, bone2Name=thigh_l, swing1LimitDegrees=45, swing2LimitDegrees=45, twistLimitDegrees=45}
 {bone1Name=thigh_twist_01_l, bone2Name=thigh_l, swing1LimitDegrees=45, swing2LimitDegrees=45, twistLimitDegrees=45}
 {
  bone1Name=lowerarm_r
  bone2Name=clavicle_r
  swing1LimitDegrees=45
  swing1Motion=Free
  swing2LimitDegrees=45
  swing2Motion=Free
  twistLimitDegrees=45
  twistMotion=Free
 }
 {
  bone1Name=clavicle_r
  bone2Name=spine_05
  swing1LimitDegrees=45
  swing1Motion=Free
  swing2LimitDegrees=45
  swing2Motion=Free
  twistLimitDegrees=45
  twistMotion=Free
 }
]
