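# Post-process style Anim Blueprint for the female large PressureGhoul character body, adding secondary motion to its attached mesh.
# It copies the parent pose, then runs a pre-sim Control Rig and weighted constraints pinning D-ring and zipper joints to body bones.
# RigidBody simulation and spring-based AnimDynamics move accessories like zippers, pouch and bandana before a post-sim Control Rig.
# The event graph updates crouch and gravity-override values, though the latter is unread here.
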AnimBlueprint /BRCosmetics/Characters/Player/Female/Large/Bodies/F_LRG_PressureGhoul/Meshes/F_LRG_PressureGhoul_AnimBP
target_skeleton=&/F_LRG_PressureGhoul_Skeleton
graphs [
 AnimGraph {nodes [
   n1 = Control Rig!(Source=Copy Pose From Mesh() {bUseAttachedParent=true}, Alpha, RBAN:bool=Get bRigidBodyEnabled(), frontend:bool=Get bIsInFrontEnd()) -> (Pose) {
    alphaBoolBlend {blendOption=Linear}
    controlRigAssetReference {blueprintRigClass=&/F_LRG_PressureGhoul_CtrlRig_presim.F_LRG_PressureGhoul_CtrlRig_presim_C}
   }
   n2 = "Constraint - pin_to_mesh_R_dring_joint (PPPPP)"!(
    ComponentPose=Local To Component(LocalPose=@n1)
    Alpha=1
    ConstraintWeights_0=0.7363
    ConstraintWeights_1:float=0.1025
    ConstraintWeights_2:float=0.0261
    ConstraintWeights_3:float=0.0118
    ConstraintWeights_4:float=0.1232
   ) -> (Pose) {
    alphaBoolBlend {blendOption=Linear}
    boneToModify {boneName=pin_to_mesh_R_dring_joint}
    constraintSetup [
     {targetBone {boneName=pelvis}}
     {targetBone {boneName=spine_01}}
     {targetBone {boneName=spine_02}}
     {targetBone {boneName=spine_03}}
     {targetBone {boneName=extra_spine_05}}
    ]
    constraintWeights [1,1,1,1,1]
   }
   n3 = "Constraint - pin_to_mesh_L_dring_joint (PPPPPP)"!(
    ComponentPose=@n2
    Alpha=1
    ConstraintWeights_0=0.736
    ConstraintWeights_1:float=0.1024
    ConstraintWeights_2:float=0.0261
    ConstraintWeights_3:float=0.0118
    ConstraintWeights_4:float=0.0356
    ConstraintWeights_5:float=0.0881
   ) -> (Pose) {
    alphaBoolBlend {blendOption=Linear}
    boneToModify {boneName=pin_to_mesh_L_dring_joint}
    constraintSetup [
     {targetBone {boneName=pelvis}}
     {targetBone {boneName=spine_01}}
     {targetBone {boneName=spine_02}}
     {targetBone {boneName=spine_03}}
     {targetBone {boneName=thigh_twist_01_l}}
     {targetBone {boneName=deform_groin_l}}
    ]
    constraintWeights [1,1,1,1,1,1]
   }
   n4 = "Constraint - pin_to_mesh__R_arm_zipper (PPPPPPPP)"!(
    ComponentPose=@n3
    Alpha=1
    ConstraintWeights_0=0.0009
    ConstraintWeights_1:float=0.0407
    ConstraintWeights_2:float=0.09
    ConstraintWeights_3:float=0.6692
    ConstraintWeights_4:float=0.0918
    ConstraintWeights_5:float=0.0201
    ConstraintWeights_6:float=0.0694
    ConstraintWeights_7:float=0.0177
   ) -> (Pose) {
    alphaBoolBlend {blendOption=Linear}
    boneToModify {boneName=pin_to_mesh__R_arm_zipper}
    constraintSetup [
     {targetBone {boneName=spine_03}}
     {targetBone {boneName=clavicle_r}}
     {targetBone {boneName=upperarm_r}}
     {targetBone {boneName=upperarm_twist_01_r}}
     {targetBone {boneName=upperarm_twist_02_r}}
     {targetBone {boneName=deform_bicep_r}}
     {targetBone {boneName=deform_upperarm_fr_r}}
     {targetBone {boneName=deform_upperarm_up_r}}
    ]
    constraintWeights [1,1,1,1,1,1,1,1]
   }
   n5 = "Constraint - pin_to_mesh_dyn_jacket_zipper_01 (PPPPPPPP)"!(
    ComponentPose=@n4
    Alpha=1
    ConstraintWeights_0=0.0216
    ConstraintWeights_1:float=0.007
    ConstraintWeights_2:float=0.0043
    ConstraintWeights_3:float=0.0085
    ConstraintWeights_4:float=0.009
    ConstraintWeights_5:float=0.0165
    ConstraintWeights_6:float=0.0104
    ConstraintWeights_7:float=0.9227
   ) -> (Pose) {
    alphaBoolBlend {blendOption=Linear}
    boneToModify {boneName=pin_to_mesh_dyn_jacket_zipper_01}
    constraintSetup [
     {targetBone {boneName=spine_03}}
     {targetBone {boneName=spine_04}}
     {targetBone {boneName=spine_05}}
     {targetBone {boneName=clavicle_l}}
     {targetBone {boneName=deform_bicep_l}}
     {targetBone {boneName=deform_pec_l}}
     {targetBone {boneName=extra_spine_02}}
     {targetBone {boneName=dyn_jacket_fr_l_01}}
    ]
    constraintWeights [1,1,1,1,1,1,1,1]
   }
   n6 = RigidBody!(ComponentPose=@n5, Alpha, OverrideWorldGravity:Vector="0,0,-980", ExternalForce) -> (Pose) {
    bClampLinearTranslationLimitToRefPose=true
    bOverrideWorldGravity=true
    overridePhysicsAsset=&/F_LRG_PressureGhoul_Physics
    simulationSpace=BaseBoneSpace
    alphaBoolBlend {blendOption=Linear}
    baseBoneRef {boneName=spine_03}
    componentAppliedLinearAccClamp=<2000 2000 2000>
    componentLinearAccScale=<0.3 0.3 0.3>
    componentLinearVelScale=<0.3 0.3 0.3>
    overrideWorldGravity {z=-980}
   }
   n7 = Anim Dynamics - Bone: dyn_R_arm_zipper(ComponentPose=@n6, Alpha=1, ExternalForce="0,0,0") {
    angularDampingOverride=1
    angularSpringConstant=20000
    bAngularSpring=true
    bOverrideAngularDamping=true
    bUseGravityOverride=true
    gravityScale=0.35
    simulationSpace=RootRelative
    alphaBoolBlend {blendOption=Linear}
    boundBone {boneName=dyn_R_arm_zipper}
    componentLinearAccScale=<1 1 1>
    componentLinearVelScale=<1 1 1>
    gravityOverride {z=-980}
    physicsBodyDefinitions [{
      boundBone {boneName=dyn_R_arm_zipper}
      boxExtents=<20 30 30>
      constraintSetup {angularLimitsMax=<5 10 2>, angularLimitsMin=<-5 -10 -120>, angularTarget=<1 0 0>}
      localJointOffset=<4 0 0>
     }]
   }
   n8 = Anim Dynamics - Bone: dyn_jacket_zipper_01(ComponentPose=@n7, Alpha=1, ExternalForce="0,0,0") {
    angularDampingOverride=1
    angularSpringConstant=20000
    bAngularSpring=true
    bOverrideAngularDamping=true
    bUseGravityOverride=true
    gravityScale=0.35
    simulationSpace=RootRelative
    alphaBoolBlend {blendOption=Linear}
    boundBone {boneName=dyn_jacket_zipper_01}
    componentLinearAccScale=<1 1 1>
    componentLinearVelScale=<1 1 1>
    gravityOverride {z=-980}
    physicsBodyDefinitions [{
      boundBone {boneName=dyn_jacket_zipper_01}
      boxExtents=<20 30 30>
      constraintSetup {angularLimitsMax=<5 10 2>, angularLimitsMin=<-5 -10 -120>, angularTarget=<1 0 0>}
      localJointOffset=<4 0 0>
     }]
   }
   n9 = Anim Dynamics - Bone: dyn_pouch_dring_joint(ComponentPose=@n8, Alpha=1, ExternalForce="0,0,0") {
    angularDampingOverride=1
    angularSpringConstant=20000
    bAngularSpring=true
    bOverrideAngularDamping=true
    bUseGravityOverride=true
    gravityScale=0.35
    simulationSpace=RootRelative
    alphaBoolBlend {blendOption=Linear}
    boundBone {boneName=dyn_pouch_dring_joint}
    componentLinearAccScale=<1 1 1>
    componentLinearVelScale=<1 1 1>
    gravityOverride {z=-980}
    physicsBodyDefinitions [{
      boundBone {boneName=dyn_pouch_dring_joint}
      boxExtents=<20 10 10>
      constraintSetup {angularLimitsMax=<5 3 2>, angularLimitsMin=<-5 -3 -120>, angularTarget=<1 0 0>}
      localJointOffset=<6 0 0>
     }]
   }
   n10 = Anim Dynamics - Bone: dyn_bandana_1(ComponentPose=@n9, Alpha=1, ExternalForce="0,0,0") {
    angularDampingOverride=1
    angularSpringConstant=20000
    bAngularSpring=true
    bOverrideAngularDamping=true
    bUseGravityOverride=true
    gravityScale=0.35
    simulationSpace=RootRelative
    alphaBoolBlend {blendOption=Linear}
    boundBone {boneName=dyn_bandana_1}
    componentLinearAccScale=<1 1 1>
    componentLinearVelScale=<1 1 1>
    gravityOverride {z=-980}
    physicsBodyDefinitions [{
      boundBone {boneName=dyn_bandana_1}
      boxExtents=<20 30 30>
      constraintSetup {angularLimitsMax=<15 15 15>, angularLimitsMin=<-15 -15 -15>, angularTarget=<1 0 0>}
      localJointOffset=<4 0 0>
     }]
   }
   n11 = Anim Dynamics - Bone: dyn_bandana_2(ComponentPose=@n10, Alpha=1, ExternalForce="0,0,0") {
    angularDampingOverride=1
    angularSpringConstant=20000
    bAngularSpring=true
    bOverrideAngularDamping=true
    bUseGravityOverride=true
    gravityScale=0.35
    simulationSpace=RootRelative
    alphaBoolBlend {blendOption=Linear}
    boundBone {boneName=dyn_bandana_2}
    componentLinearAccScale=<1 1 1>
    componentLinearVelScale=<1 1 1>
    gravityOverride {z=-980}
    physicsBodyDefinitions [{
      boundBone {boneName=dyn_bandana_2}
      boxExtents=<20 30 30>
      constraintSetup {angularLimitsMax=<15 15 15>, angularLimitsMin=<-15 -15 -15>, angularTarget=<1 0 0>}
      localJointOffset=<4 0 0>
     }]
   }
   n12 = Anim Dynamics - Bone: dyn_bandana_3(ComponentPose=@n11, Alpha=1, ExternalForce="0,0,0") {
    angularDampingOverride=1
    angularSpringConstant=20000
    bAngularSpring=true
    bOverrideAngularDamping=true
    bUseGravityOverride=true
    gravityScale=0.35
    simulationSpace=BoneRelative
    alphaBoolBlend {blendOption=Linear}
    boundBone {boneName=dyn_bandana_3}
    componentLinearAccScale=<1 1 1>
    componentLinearVelScale=<1 1 1>
    gravityOverride {z=-980}
    physicsBodyDefinitions [{
      boundBone {boneName=dyn_bandana_3}
      boxExtents=<20 30 30>
      constraintSetup {angularLimitsMax=<15 15 15>, angularLimitsMin=<-15 -15 -15>, angularTarget=<1 0 0>}
      localJointOffset=<4 0 0>
     }]
    relativeSpaceBone {boneName=spine_05}
   }
   n13 = Anim Dynamics - Bone: dyn_bandana_4(ComponentPose=@n12, Alpha=1, ExternalForce="0,0,0") {
    angularDampingOverride=1
    angularSpringConstant=20000
    bAngularSpring=true
    bOverrideAngularDamping=true
    bUseGravityOverride=true
    gravityScale=0.35
    simulationSpace=BoneRelative
    alphaBoolBlend {blendOption=Linear}
    boundBone {boneName=dyn_bandana_4}
    componentLinearAccScale=<1 1 1>
    componentLinearVelScale=<1 1 1>
    gravityOverride {z=-980}
    physicsBodyDefinitions [{
      boundBone {boneName=dyn_bandana_4}
      boxExtents=<20 30 30>
      constraintSetup {angularLimitsMax=<15 15 15>, angularLimitsMin=<-15 -15 -15>, angularTarget=<1 0 0>}
      localJointOffset=<4 0 0>
     }]
    relativeSpaceBone {boneName=spine_05}
   }
   n14 = Control Rig!(
    Source=Component To Local(ComponentPose=@n13)
    Alpha
    crouching:double=Get crouching_float() -> (crouching_float:double)
    arms_downwards:double=Get arms_downwards() -> (arms_downwards:double)
    hands_on_belt:double=Get hands_on_belt() -> (hands_on_belt:double)
   ) -> (Pose) {
    alphaBoolBlend {blendOption=Linear}
    controlRigAssetReference {blueprintRigClass=&/F_LRG_PressureGhoul_CtrlRig_postsim.F_LRG_PressureGhoul_CtrlRig_postsim_C}
   }
   Output Pose(Result=@n14) {layerGroup=None}
  ]}
 EventGraph {nodes [
   Event BlueprintUpdateAnimation()
   > n1 = Gravity Override State Function DEPRECATED(CharacterPartAnimInstance=Self-Reference(), GameTimeInSeconds=GetGameTimeInSeconds(), GravityOverrideParams=Get Gravity Override Params())
   > Set Gravity Override(Gravity Override=@n1)
   > Set crouching_float(crouching_float:double="To Float (Boolean)"(InBool=Get bIsCrouching())) -> (Output_Get:double)
   Event OnCharacterCustomizationLoaded()
  ]}
]
variables [
 arms_downwards {type=number}
 crouching_float {type=number}
 Gravity Override {type=Vector}
 Gravity Override Params {type=GravityOverrideParamsStruct}
 hands_on_belt {type=number}
 skirt_on/off {type=boolean}
]
