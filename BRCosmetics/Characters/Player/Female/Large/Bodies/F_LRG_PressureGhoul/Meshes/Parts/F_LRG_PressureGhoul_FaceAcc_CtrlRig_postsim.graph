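# Post-sim Control Rig for the PressureGhoul face accessory that nudges the right front side-hair bone in response to the right big collar's motion.
# It reads dyn_big_collar_r_01 relative to its base bone and remaps the Euler Y rotation into a 0–1 weight.
# That weight drives a Modify Transforms offset along X on extra_f_side_hair_02_r.
# The left side has no logic, and the sky_diving variable is unused.
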ControlRigBlueprint /BRCosmetics/Characters/Player/Female/Large/Bodies/F_LRG_PressureGhoul/Meshes/Parts/F_LRG_PressureGhoul_FaceAcc_CtrlRig_postsim
graphs [
 RigVMFunctionLibrary {links [], nodes []}
 RigVMFunction_Sequence_1_ContainedGraph {nodes [
   n1 = RigVMFunctionEntryNode()
   RigVMFunction_Sequence = &RigVMFunction_Sequence(ExecuteContext=@n1.ExecuteContext)
   n2 = &RigVMFunction_Sequence(ExecuteContext=@RigVMFunction_Sequence.B)
   n3 = &RigVMFunction_Sequence(ExecuteContext=@n2.B)
   n4 = &RigVMFunction_Sequence(ExecuteContext=@n3.B)
   n5 = &RigVMFunction_Sequence(ExecuteContext=@n4.B)
   n6 = &RigVMFunction_Sequence(ExecuteContext=@n5.B)
   n7 = &RigVMFunction_Sequence(ExecuteContext=@n6.B)
   RigVMFunctionReturnNode(A=@RigVMFunction_Sequence.A, B=@n2.A, C=@n3.A, D=@n4.A, E=@n5.A, F=@n6.A, G=@n7.A, H=@n7.B)
  ]}
 RigVMModel {nodes [
   RigUnit_BeginExecution = &RigUnit_BeginExecution()
   RigVMFunction_Sequence = RigVMAggregateNode(ExecuteContext=@RigUnit_BeginExecution.ExecutePin)
   n1 = &RigUnit_GetRelativeTransformForItem(Child=(Type=Bone,Name="dyn_big_collar_r_01"), Parent=(Type=Bone,Name="BASE_dyn_collar_r_01"), bChildInitial=False, bParentInitial=false)
   &RigUnit_ModifyTransforms(
    ItemToModify=((Item=(Type=Bone,Name="extra_f_side_hair_02_r"),Transform=(Rotation=(X=0,Y=0,Z=0,W=1),Translation=(X=-1,Y=0,Z=0),Scale3D=(X=1,Y=1,Z=1))))
    Weight=&RigVMFunction_MathDoubleRemap(
     SourceMaximum=-40
     SourceMinimum=0
     TargetMaximum=1
     TargetMinimum=0
     Value=&RigVMFunction_MathQuaternionToEuler(RotationOrder=ZYX, Value=@n1.RelativeTransform.Rotation).Result.Y
    ).Result
    ExecutePin=@RigVMFunction_Sequence.A
   )
  ]}
]
hierarchy [
 root {loc=<0 0 0>, rot=<0 0 0 1>}
 pelvis {parent=root, loc=<0 -0.648 94.7699>, rot=<0.0127 0.707 -0.0127 -0.707>}
 spine_01 {parent=pelvis, loc=<5.0975 -0.6388 0>, rot=<0 0 0.034 -0.9994>}
 spine_02 {parent=spine_01, loc=<7.252 -0.035 0>, rot=<0 0 0.0238 -0.9997>}
 spine_03 {parent=spine_02, loc=<6.0106 0 0>, rot=<0 0 0.0149 -0.9999>}
 spine_04 {parent=spine_03, loc=<8.1024 0 0>, rot=<0 0 -0.0132 -0.9999>}
 spine_05 {parent=spine_04, loc=<10.0142 0 0>, rot=<0 0 -0.0015 -1>}
 neck_01 {parent=spine_05, loc=<11.3775 -0.6396 0>, rot=<0 0 -0.2003 -0.9797>}
 neck_02 {parent=neck_01, loc=<4.5627 0 0>, rot=<0 0 0.0504 -0.9987>}
 head {parent=neck_02, loc=<5.3914 0 0>, rot=<0 0 0.115 -0.9934>}
 faceAttach {parent=head, loc=<-0.8861 -0.4209 -0.0906>, rot=<0.0082 -0.7071 -0.0082 -0.7071>, scale=<0.89 0.89 0.89>}
 dyn_f_side_hair_BASE_r {parent=head, loc=<6.1807 0.9122 7.9582>, rot=<0.04 -0.6543 0.7537 -0.0461>}
 dyn_f_side_hair_01_r {parent=dyn_f_side_hair_BASE_r, loc=<0 0 0>, rot=<0 0 0 1>}
 dyn_f_side_hair_02_r {parent=dyn_f_side_hair_01_r, loc=<9.6346 -0.2364 0>, rot=<0 0 0.0828 -0.9966>}
 extra_f_side_hair_02_r {parent=dyn_f_side_hair_02_r, loc=<0 0 0>, rot=<0 0 0 1>}
 dyn_side_hair_BASE_r {parent=head, loc=<6.9517 -5.3725 7.4638>, rot=<0.061 -0.9981 0.0021 -0.0001>}
 dyn_side_hair_01_r {parent=dyn_side_hair_BASE_r, loc=<0 0 0>, rot=<0 0 0 1>}
 dyn_side_hair_02_r {parent=dyn_side_hair_01_r, loc=<8.2212 0 0>, rot=<0 0 0.0828 -0.9966>}
 dyn_side_hair_03_r {parent=dyn_side_hair_02_r, loc=<8.2594 0 0>, rot=<0 0 0.0338 -0.9994>}
 dyn_back_hair_BASE_m {parent=head, loc=<7.3916 -8.9584 0.3833>, rot=<0.061 -0.9981 0 0>}
 dyn_back_hair_01_m {parent=dyn_back_hair_BASE_m, loc=<0 0 0>, rot=<0 0 0 1>}
 dyn_back_hair_02_m {parent=dyn_back_hair_01_m, loc=<8.2212 0 0>, rot=<0 0 0.0828 -0.9966>}
 dyn_back_hair_03_m {parent=dyn_back_hair_02_m, loc=<8.2594 0 0>, rot=<0 0 0.0338 -0.9994>}
 dyn_f_side_hair_BASE_l {parent=head, loc=<5.9348 1.5 -7.7525>, rot=<0 -1 0 0>}
 dyn_f_side_hair_01_l {parent=dyn_f_side_hair_BASE_l, loc=<0 0 0>, rot=<0 0 0 1>}
 dyn_f_side_hair_02_l {parent=dyn_f_side_hair_01_l, loc=<9.6346 -0.2364 0>, rot=<0 0 0 1>}
 extra_f_side_hair_02_l {parent=dyn_f_side_hair_02_l, loc=<0 0 0>, rot=<0 0 0 1>}
 dyn_side_hair_BASE_l {parent=head, loc=<7.1193 -6.7394 -7.0822>, rot=<0.0609 -0.9961 -0.0644 0.0039>}
 dyn_side_hair_01_l {parent=dyn_side_hair_BASE_l, loc=<0 0 0>, rot=<0 0 0 1>}
 dyn_side_hair_02_l {parent=dyn_side_hair_01_l, loc=<8.2212 0 0>, rot=<0 0 0.0828 -0.9966>}
 dyn_side_hair_03_l {parent=dyn_side_hair_02_l, loc=<8.2594 0 0>, rot=<0 0 0.0338 -0.9994>}
 FACIAL_C_FacialRoot {parent=head, loc=<2.1458 0.049 -0.0002>, rot=<0.4879 -0.5118 -0.4879 -0.5118>}
 FACIAL_C_Forehead {parent=FACIAL_C_FacialRoot, loc=<-0.0009 -3.2724 7.6822>, rot=<-0.0584 0 0 -0.9983>}
 FACIAL_L_ForeheadIn {parent=FACIAL_C_FacialRoot, loc=<1.7535 -3.265 7.5642>, rot=<-0.0756 -0.0697 0.0032 -0.9947>}
 FACIAL_R_ForeheadIn {parent=FACIAL_C_FacialRoot, loc=<-1.757 -3.2673 7.5612>, rot=<-0.0756 0.0697 -0.0032 -0.9947>}
 FACIAL_L_ForeheadMid {parent=FACIAL_C_FacialRoot, loc=<3.2411 -3.5408 7.0994>, rot=<-0.0857 -0.1753 -0.0232 -0.9805>}
 FACIAL_R_ForeheadMid {parent=FACIAL_C_FacialRoot, loc=<-3.2443 -3.5432 7.097>, rot=<-0.0857 0.1753 0.0232 -0.9805>}
 FACIAL_L_ForeheadOut {parent=FACIAL_C_FacialRoot, loc=<4.5538 -3.4807 6.2291>, rot=<-0.1357 -0.2544 -0.083 -0.9539>}
 FACIAL_R_ForeheadOut {parent=FACIAL_C_FacialRoot, loc=<-4.5557 -3.4829 6.2293>, rot=<-0.1357 0.2544 0.083 -0.9539>}
 FACIAL_L_EyesackUpper {parent=FACIAL_C_FacialRoot, loc=<3.3236 -2.5134 7.5768>, rot=<0.034 -0.0833 -0.0658 -0.9938>}
 FACIAL_R_EyesackUpper {parent=FACIAL_C_FacialRoot, loc=<-3.3237 -2.5117 7.5772>, rot=<0.034 0.0833 0.0658 -0.9938>}
 FACIAL_L_EyelidUpperFurrow {parent=FACIAL_C_FacialRoot, loc=<3.2191 -2.3002 7.4847>, rot=<-0.0448 -0.1496 -0.0819 -0.9843>}
 FACIAL_R_EyelidUpperFurrow {parent=FACIAL_C_FacialRoot, loc=<-3.2144 -2.3006 7.4824>, rot=<-0.0448 0.1496 0.0819 -0.9843>}
 FACIAL_L_EyelidUpperA {parent=FACIAL_C_FacialRoot, loc=<2.9621 -1.3556 5.3844>, rot=<-0.1626 0 0 -0.9867>}
 FACIAL_R_EyelidUpperA {parent=FACIAL_C_FacialRoot, loc=<-2.909 -1.3552 5.3941>, rot=<-0.1626 0 0 -0.9867>}
 FACIAL_L_Eye {parent=FACIAL_C_FacialRoot, loc=<2.9463 -1.3866 5.3198>, rot=<-0.015 0 0 -0.9999>}
 FACIAL_R_Eye {parent=FACIAL_C_FacialRoot, loc=<-2.8977 -1.3844 5.324>, rot=<-0.015 0 0 -0.9999>}
 FACIAL_L_EyelidLowerA {parent=FACIAL_C_FacialRoot, loc=<2.9621 -1.3556 5.3844>, rot=<0.2696 0 0 -0.963>}
 FACIAL_R_EyelidLowerA {parent=FACIAL_C_FacialRoot, loc=<-2.909 -1.3552 5.3941>, rot=<0.2696 0 0 -0.963>}
 FACIAL_L_EyeCornerInner {parent=FACIAL_C_FacialRoot, loc=<1.6115 -0.6424 6.6972>, rot=<-0.015 0 0 -0.9999>}
 FACIAL_R_EyeCornerInner {parent=FACIAL_C_FacialRoot, loc=<-1.6012 -0.6614 6.6927>, rot=<-0.015 0 0 -0.9999>}
 FACIAL_L_EyeCornerOuter {parent=FACIAL_C_FacialRoot, loc=<4.603 -1.3331 6.2221>, rot=<-0.0138 -0.3825 -0.0057 -0.9238>}
 FACIAL_R_EyeCornerOuter {parent=FACIAL_C_FacialRoot, loc=<-4.5983 -1.3229 6.2213>, rot=<-0.0138 0.3826 0.0057 -0.9238>}
 FACIAL_L_EyesackLower {parent=FACIAL_C_FacialRoot, loc=<3.5446 0.024 6.4594>, rot=<-0.0371 -0.1394 -0.0243 -0.9892>}
 FACIAL_R_EyesackLower {parent=FACIAL_C_FacialRoot, loc=<-3.5367 0.0273 6.4606>, rot=<-0.0371 0.1394 0.0244 -0.9892>}
 FACIAL_L_CheekInner {parent=FACIAL_C_FacialRoot, loc=<2.1539 0.966 6.5314>, rot=<-0.1377 -0.085 -0.1984 -0.9667>}
 FACIAL_R_CheekInner {parent=FACIAL_C_FacialRoot, loc=<-2.1508 0.9613 6.5299>, rot=<-0.1377 0.085 0.1984 -0.9667>}
 FACIAL_L_CheekOuter {parent=FACIAL_C_FacialRoot, loc=<4.0457 1.0054 5.0144>, rot=<0.0414 -0.3784 0.0947 -0.9199>}
 FACIAL_R_CheekOuter {parent=FACIAL_C_FacialRoot, loc=<-4.0469 0.999 5.0139>, rot=<0.0414 0.3784 -0.0947 -0.9199>}
 FACIAL_L_NasolabialBulge {parent=FACIAL_C_FacialRoot, loc=<3.3528 4.2875 4.6934>, rot=<0.0441 -0.3776 0.0182 -0.9247>}
 FACIAL_R_NasolabialBulge {parent=FACIAL_C_FacialRoot, loc=<-3.3515 4.2875 4.6928>, rot=<0.0441 0.3776 -0.0182 -0.9247>}
 FACIAL_L_NasolabialFurrow {parent=FACIAL_C_FacialRoot, loc=<2.8562 3.0343 6.9903>, rot=<-0.0471 -0.1204 -0.3655 -0.9218>}
 FACIAL_R_NasolabialFurrow {parent=FACIAL_C_FacialRoot, loc=<-2.8568 3.0335 6.9903>, rot=<-0.0471 0.1204 0.3655 -0.9218>}
 FACIAL_L_CheekLower {parent=FACIAL_C_FacialRoot, loc=<3.7167 4.1908 4.183>, rot=<0.0211 -0.4277 0.0514 -0.9022>}
 FACIAL_R_CheekLower {parent=FACIAL_C_FacialRoot, loc=<-3.7149 4.1913 4.1837>, rot=<0.0211 0.4277 -0.0513 -0.9022>}
 FACIAL_C_Nose {parent=FACIAL_C_FacialRoot, loc=<0.0004 0.4062 7.8968>, rot=<-0.1452 0 0 -0.9894>}
 FACIAL_C_NoseLower {parent=FACIAL_C_Nose, loc=<-0.0002 2.776 -0.463>, rot=<0.1305 0 0 -0.9914>}
 FACIAL_L_Nostril {parent=FACIAL_C_Nose, loc=<1.0456 2.0493 -0.1952>, rot=<0.0173 0.1305 0.0023 -0.9913>}
 FACIAL_R_Nostril {parent=FACIAL_C_Nose, loc=<-1.0411 2.0421 -0.2043>, rot=<0.0173 -0.1305 -0.0023 -0.9913>}
 FACIAL_C_NoseTip {parent=FACIAL_C_Nose, loc=<-0.0012 2.1796 0.5534>, rot=<0.0872 0 0 -0.9962>}
 FACIAL_C_MouthUpper {parent=FACIAL_C_FacialRoot, loc=<0.0003 4.5274 5.6217>, rot=<-0.0149 0 0 -0.9999>}
 FACIAL_C_LipUpper {parent=FACIAL_C_MouthUpper, loc=<0.0029 -0.1214 1.9808>, rot=<-0.0872 0 0 -0.9962>}
 FACIAL_L_LipUpper {parent=FACIAL_C_MouthUpper, loc=<1.1804 -0.0962 1.7714>, rot=<-0.0778 -0.1601 -0.0379 -0.9833>}
 FACIAL_R_LipUpper {parent=FACIAL_C_MouthUpper, loc=<-1.1775 -0.0976 1.7749>, rot=<-0.0778 0.1601 0.0379 -0.9833>}
 FACIAL_L_LipUpperOuter {parent=FACIAL_C_MouthUpper, loc=<2.0734 -0.0208 1.1797>, rot=<-0.0984 -0.2293 -0.0983 -0.9634>}
 FACIAL_R_LipUpperOuter {parent=FACIAL_C_MouthUpper, loc=<-2.0705 -0.0196 1.1786>, rot=<-0.0984 0.2293 0.0983 -0.9634>}
 FACIAL_L_LipCorner {parent=FACIAL_C_MouthUpper, loc=<2.5143 0.2722 0.6014>, rot=<0 -0.2078 0 -0.9782>}
 FACIAL_R_LipCorner {parent=FACIAL_C_MouthUpper, loc=<-2.512 0.2716 0.6003>, rot=<0 0.2078 0 -0.9782>}
 FACIAL_L_Masseter {parent=FACIAL_C_FacialRoot, loc=<6.2681 1.7253 3.1356>, rot=<0.0545 0.1128 0.0096 -0.9921>}
 FACIAL_R_Masseter {parent=FACIAL_C_FacialRoot, loc=<-6.2679 1.7207 3.137>, rot=<0.0545 -0.1128 -0.0096 -0.9921>}
 FACIAL_C_UnderChin {parent=FACIAL_C_FacialRoot, loc=<-0.0005 8.7597 2.9166>, rot=<0.4868 0 0 -0.8735>}
 FACIAL_L_UnderChin {parent=FACIAL_C_FacialRoot, loc=<3.0384 7.557 1.8038>, rot=<0.3899 -0.3561 0.077 -0.8457>}
 FACIAL_R_UnderChin {parent=FACIAL_C_FacialRoot, loc=<-3.037 7.5515 1.807>, rot=<0.3899 0.356 -0.077 -0.8457>}
 FACIAL_C_TeethUpper {parent=FACIAL_C_FacialRoot, loc=<-0.0012 5.0466 6.5285>, rot=<-0.0149 0 0 -0.9999>}
 FACIAL_C_LowerLipRotation {parent=FACIAL_C_FacialRoot, loc=<0.0024 2.6751 1.031>, rot=<0.328 0 0 -0.9447>}
 FACIAL_C_MouthLower {parent=FACIAL_C_LowerLipRotation, loc=<-0.0022 -1.3907 4.7509>, rot=<-0.342 0 0 -0.9397>}
 FACIAL_C_LipLower {parent=FACIAL_C_MouthLower, loc=<-0.0007 1.0865 1.3675>, rot=<0.0872 0 0 -0.9962>}
 FACIAL_L_LipLower {parent=FACIAL_C_MouthLower, loc=<1.0889 0.9546 1.2684>, rot=<0.0927 -0.1294 0.0272 -0.9869>}
 FACIAL_R_LipLower {parent=FACIAL_C_MouthLower, loc=<-1.0921 0.9571 1.2682>, rot=<0.0927 0.1294 -0.0272 -0.9869>}
 FACIAL_L_LipLowerOuter {parent=FACIAL_C_MouthLower, loc=<2.0028 0.6114 0.922>, rot=<0.0913 -0.2321 0.0318 -0.9679>}
 FACIAL_R_LipLowerOuter {parent=FACIAL_C_MouthLower, loc=<-2.005 0.6086 0.9266>, rot=<0.0913 0.2321 -0.0318 -0.9679>}
 FACIAL_C_TeethLower {parent=FACIAL_C_LowerLipRotation, loc=<-0.0037 -1.4883 5.7273>, rot=<-0.342 0 0 -0.9397>}
 FACIAL_C_Tongue1 {parent=FACIAL_C_TeethLower, loc=<0.0016 -0.0532 -4.4917>, rot=<-0.2588 0 0 -0.9659>}
 FACIAL_C_Tongue2 {parent=FACIAL_C_Tongue1, loc=<-0.0001 0.4278 1.2438>, rot=<0.2588 0 0 -0.9659>}
 FACIAL_C_Tongue3 {parent=FACIAL_C_Tongue2, loc=<-0.0002 0.216 1.1684>, rot=<0.1305 0 0 -0.9914>}
 FACIAL_C_Tongue4 {parent=FACIAL_C_Tongue3, loc=<0.0004 -0.1365 1.0837>, rot=<0.0436 0 0 -0.999>}
 FACIAL_C_TongueUpper3 {parent=FACIAL_C_Tongue3, loc=<0.0007 -0.3416 -0.1537>, rot=<0 0 0 1>}
 FACIAL_C_TongueLower3 {parent=FACIAL_C_Tongue3, loc=<-0.0007 0.8276 0.0149>, rot=<0 0 0 1>}
 FACIAL_L_TongueSide3 {parent=FACIAL_C_Tongue3, loc=<1.5363 0.003 0.1167>, rot=<0 0 0 1>}
 FACIAL_R_TongueSide3 {parent=FACIAL_C_Tongue3, loc=<-1.5353 0.0004 0.113>, rot=<0 0 0 1>}
 FACIAL_C_TongueUpper2 {parent=FACIAL_C_Tongue2, loc=<0.0012 -0.5836 -0.1142>, rot=<0 0 0 1>}
 FACIAL_L_TongueSide2 {parent=FACIAL_C_Tongue2, loc=<1.8122 0.078 0.1692>, rot=<0 0 0 1>}
 FACIAL_R_TongueSide2 {parent=FACIAL_C_Tongue2, loc=<-1.8119 0.0733 0.1646>, rot=<0 0 0 1>}
 FACIAL_C_Jaw {parent=FACIAL_C_FacialRoot, loc=<0.0024 2.6751 1.031>, rot=<0.328 0 0 -0.9447>}
 FACIAL_C_Jawline {parent=FACIAL_C_Jaw, loc=<-0.003 2.3347 6.5331>, rot=<0.2586 0 0 -0.966>}
 FACIAL_L_ChinSide {parent=FACIAL_C_Jaw, loc=<2.7691 0.3301 6.176>, rot=<-0.2208 -0.3738 0.0499 -0.8995>}
 FACIAL_R_ChinSide {parent=FACIAL_C_Jaw, loc=<-2.7791 0.3311 6.1735>, rot=<-0.2208 0.3738 -0.0499 -0.8995>}
 FACIAL_C_Chin {parent=FACIAL_C_Jaw, loc=<-0.0044 0.9561 6.4363>, rot=<-0.2591 0 0 -0.9659>}
 FACIAL_L_Ear {parent=FACIAL_C_FacialRoot, loc=<6.7569 0.119 0.4806>, rot=<-0.0412 -0.0003 -0.0262 -0.9988>}
 FACIAL_R_Ear {parent=FACIAL_C_FacialRoot, loc=<-6.7525 0.1219 0.4716>, rot=<-0.0412 0.0003 0.0262 -0.9988>}
 FACIAL_L_Jawline {parent=FACIAL_C_FacialRoot, loc=<4.3601 5.5985 3.2557>, rot=<0.1949 0.1585 -0.0523 -0.9665>}
 FACIAL_R_Jawline {parent=FACIAL_C_FacialRoot, loc=<-4.3592 5.598 3.2533>, rot=<0.1949 -0.1585 0.0523 -0.9665>}
 FACIAL_C_NoseBridge {parent=FACIAL_C_FacialRoot, loc=<-0.0002 -1.645 8.4219>, rot=<0.0289 0 0 -0.9996>}
 FACIAL_L_EyelidUpperB {parent=FACIAL_C_FacialRoot, loc=<2.9621 -1.3556 5.3844>, rot=<-0.2395 0 0 -0.9709>}
 FACIAL_L_EyelidLowerB {parent=FACIAL_C_FacialRoot, loc=<2.9621 -1.3556 5.3844>, rot=<0.3031 0 0 -0.953>}
 FACIAL_R_EyelidLowerB {parent=FACIAL_C_FacialRoot, loc=<-2.909 -1.3552 5.3941>, rot=<0.3114 0 0 -0.9503>}
 FACIAL_R_EyelidUpperB {parent=FACIAL_C_FacialRoot, loc=<-2.909 -1.3552 5.3941>, rot=<-0.2395 0 0 -0.9709>}
 clavicle_l {parent=spine_05, loc=<4.1651 -0.3023 -5.0942>, rot=<0.0465 -0.7289 -0.009 -0.6829>}
 upperarm_l {parent=clavicle_l, loc=<11.2117 -1.1486 -0.0332>, rot=<0.0179 -0.2982 -0.028 -0.9539>}
 clavicle_r {parent=spine_05, loc=<4.1652 -0.3023 5.0942>, rot=<0.7289 0.0465 -0.6829 0.009>}
 upperarm_r {parent=clavicle_r, loc=<-11.2117 1.1486 0.0332>, rot=<0.0179 -0.2982 -0.028 -0.9539>}
 extra_spine_05 {parent=spine_05, loc=<0 0 0>, rot=<0 0 0 1>}
 BASE_dyn_collar_l_01 {parent=spine_05, loc=<8.0419 4.2611 -9.9821>, rot=<0.3154 -0.4537 -0.168 -0.8163>}
 dyn_big_collar_l_01 {parent=BASE_dyn_collar_l_01, loc=<0 0 0>, rot=<0 0 0 1>}
 BASE_dyn_collar_r_01 {parent=spine_05, loc=<8.3745 3.6577 9.822>, rot=<0.7989 0.1918 0.4784 -0.3101>}
 dyn_big_collar_r_01 {parent=BASE_dyn_collar_r_01, loc=<0 0 0>, rot=<0 0 0 1>}
]
variables [sky_diving {type=bool, default=False}]
