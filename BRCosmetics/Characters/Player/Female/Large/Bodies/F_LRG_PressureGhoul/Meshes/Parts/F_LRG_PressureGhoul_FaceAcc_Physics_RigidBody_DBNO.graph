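# Physics asset for the PressureGhoul female large face accessory, the RigidBody variant named for the DBNO (down but not out) state, covering hair and collar.
# Capsules on spine, neck, head, clavicles and upper arms form the body volume.
# Tiny BASE bodies root angle-limited hair chains at the head and boxed collar flaps at spine_05.
# Angularly free cross-constraints tie both side-hair tips to the back-hair tip, linking those chains.
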PhysicsAsset /BRCosmetics/Characters/Player/Female/Large/Bodies/F_LRG_PressureGhoul/Meshes/Parts/F_LRG_PressureGhoul_FaceAcc_Physics_RigidBody_DBNO
bodies {
 BASE_dyn_collar_l_01 [{shapeType=Capsule, length=0.505, radius=0.505, shapeName=BASE_dyn_collar_l_01_capsule, rotation=<rot 0 -90 0>}]
 BASE_dyn_collar_r_01 [{shapeType=Capsule, length=0.505, radius=0.505, shapeName=BASE_dyn_collar_r_01_capsule, rotation=<rot 0 -90 0>}]
 clavicle_l [{shapeType=Capsule, length=6, radius=6, shapeName=clavicle_l_capsule, center=<1.8251 0 0>, rotation=<rot 0 -90 90>}]
 clavicle_r [{shapeType=Capsule, length=8.8021, radius=7.8535, shapeName=clavicle_r_capsule, center=<-5 1.3618 1.1802>, rotation=<rot 0 90 90>}]
 dyn_back_hair_01_m [{shapeType=Capsule, length=2.9397, radius=2.2581, shapeName=dyn_hair_back_m_00_capsule, center=<2.4009 -0.9696 0>, rotation=<rot 0 -90 0>}]
 dyn_back_hair_02_m [{shapeType=Capsule, length=6.4257, radius=2.9941, shapeName=dyn_hair_back_m_01_capsule, center=<2.1513 -2.0033 0>, rotation=<rot 0 -90 0>}]
 dyn_back_hair_03_m [{
   shapeType=Capsule
   length=3.2185
   radius=3.3324
   shapeName=dyn_hair_back_m_02_capsule
   center=<5.1852 -2.0033 0.0735>
   rotation=<rot 0 -89.9997 -90.8122>
  }]
 dyn_back_hair_BASE_m [{shapeType=Capsule, length=0.505, radius=0.505, shapeName=dyn_back_hair_BASE_m_capsule, rotation=<rot 0 -90 0>}]
 dyn_big_collar_l_01 [{
   shapeType=Box
   extentX=18.6204
   extentY=16.1603
   extentZ=2.6916
   shapeName=dyn_big_collar_l_01_box
   center=<1.1067 -1.5071 -0.4736>
   rotation=<rot 3.3022 52.1895 -8.8995>
  }]
 dyn_big_collar_r_01 [{
   shapeType=Box
   extentX=11.3444
   extentY=14.7158
   extentZ=2.4756
   shapeName=dyn_big_collar_l_01_box
   center=<2.9811 0.7017 0.0876>
   rotation=<rot -2.287 -46.8154 -179.9911>
  }]
 dyn_f_side_hair_01_l [{
   shapeType=Capsule
   length=6.6083
   radius=5.6037
   shapeName=dyn_f_side_hair_01_l_capsule
   center=<4.1903 0.0805 4.0894>
   rotation=<rot 0 52.3082 -74.2891>
  }]
 dyn_f_side_hair_01_r [{
   shapeType=Capsule
   length=6.2733
   radius=3.9439
   shapeName=dyn_f_side_hair_01_r_capsule
   center=<6.509 -2.8594 -0.4049>
   rotation=<rot -6.1053 -90 4.9595>
  }]
 dyn_f_side_hair_02_l [{
   shapeType=Capsule
   length=6.7757
   radius=4.5786
   shapeName=dyn_f_side_hair_02_l_capsule
   center=<1.9508 -1.7846 0.5943>
   rotation=<rot -2.3807 168.838 -93.489>
  }]
 dyn_f_side_hair_02_r [{
   shapeType=Capsule
   length=4.9442
   radius=3.2903
   shapeName=dyn_f_side_hair_02_r_capsule
   center=<2.6958 -0.3387 -0.387>
   rotation=<rot -12.7144 -90 -10.6368>
  }]
 dyn_f_side_hair_BASE_l [{shapeType=Capsule, length=0.505, radius=0.505, shapeName=dyn_f_side_hair_BASE_l_capsule, rotation=<rot 0 -90 0>}]
 dyn_f_side_hair_BASE_r [{shapeType=Capsule, length=0.505, radius=0.505, shapeName=dyn_f_side_hair_BASE_r_capsule, rotation=<rot 0 -90 0>}]
 dyn_side_hair_01_l [{
   shapeType=Capsule
   length=3.0925
   radius=2.3948
   shapeName=dyn_hair_back_l_00_capsule
   center=<4.2988 -0.1757 -0.3631>
   rotation=<rot 27.5105 -89.9999 -86.9923>
  }]
 dyn_side_hair_01_r [{
   shapeType=Capsule
   length=3.3697
   radius=2.4275
   shapeName=dyn_hair_back_r_00_capsule
   center=<3.3779 -0.4236 -0.0172>
   rotation=<rot 22.7063 -86.8002 85.9192>
  }]
 dyn_side_hair_02_l [
  {shapeType=Capsule, length=1.5676, radius=2.0003, center=<2.2057 -0.5796 -0.0678>, rotation=<rot 0 0.0001 33.6178>}
  {
   shapeType=Capsule
   length=1.0676
   radius=2.5003
   shapeName=dyn_side_hair_02_l_capsule
   center=<1.4053 1.3006 6.9582>
   rotation=<rot 9.4285 -17.7187 -62.8551>
  }
 ]
 dyn_side_hair_02_r [{shapeType=Capsule, length=2.273, radius=5.3, center=<6.02 -2.07 0.737>, rotation=<rot 69.9417 90.0009 -100>}]
 dyn_side_hair_03_l [{shapeType=Capsule, length=2.4059, radius=1.7804, center=<1.5 0 0.7>, rotation=<rot 0 0 17.3936>}]
 dyn_side_hair_03_r [{shapeType=Capsule, length=4.3141, radius=2.8012, center=<2.6757 -1.1485 0.1486>, rotation=<rot 0 -179.9987 -2.006>}]
 dyn_side_hair_BASE_l [{shapeType=Capsule, length=0.505, radius=0.505, shapeName=dyn_side_hair_BASE_l_capsule, rotation=<rot 0 -90 0>}]
 dyn_side_hair_BASE_r [{shapeType=Capsule, length=0.505, radius=0.505, shapeName=dyn_side_hair_BASE_r_capsule, rotation=<rot 0 -90 0>}]
 extra_spine_05 [{shapeType=Capsule, length=33.1, radius=13.5568, center=<0.2848 2.7951 0>}]
 head [
  {shapeType=Capsule, length=6.5345, radius=6.5559, center=<3.1811 2.6041 0.3139>, rotation=<rot 0.5016 -90.415 -88.3249>}
  {
   shapeType=Capsule
   length=3.3888
   radius=6.0811
   shapeName=head_capsule
   center=<1.6133 -0.0375 -0.0239>
   rotation=<rot -89.1519 179.9981 -180>
  }
  {
   shapeType=Capsule
   length=14.5345
   radius=12.5559
   shapeName=head_capsule_keep
   center=<13.465 -26.4327 0.5537>
   rotation=<rot 0.5016 -90.4152 -178.3249>
  }
 ]
 neck_02 [{shapeType=Capsule, length=0, radius=5.6134, shapeName=neck_02_capsule, rotation=<rot 0 -90 0>}]
 spine_03 [{shapeType=Capsule, length=10.5022, radius=15.3604, center=<0 2.0135 0.0001>}]
 spine_04 [{shapeType=Capsule, length=7.2641, radius=15.6748, center=<0 4.4958 0>}]
 spine_05 [
  {shapeType=Capsule, length=6.1198, radius=3.5313, center=<15.7404 -5.7066 0>}
  {shapeType=Capsule, length=26.6198, radius=16.5313, shapeName=spine_05_capsule, center=<2.4881 14.4811 0>}
 ]
 upperarm_l [{shapeType=Capsule, length=20.9787, radius=5.8924, center=<8.675 0.2066 0.3551>, rotation=<rot 0 -82.7612 -89.1886>}]
 upperarm_r [{shapeType=Capsule, length=28.169, radius=6.2045, center=<-14.7567 2.1212 -0.1158>, rotation=<rot -14.813 -92.5958 -92.3309>}]
}
constraints [
 {
  bone1Name=spine_04
  bone2Name=spine_03
  swing1LimitDegrees=45
  swing1Motion=Free
  swing2LimitDegrees=45
  swing2Motion=Free
  twistLimitDegrees=45
  twistMotion=Free
 }
 {
  bone1Name=spine_05
  bone2Name=spine_04
  swing1LimitDegrees=45
  swing1Motion=Free
  swing2LimitDegrees=45
  swing2Motion=Free
  twistLimitDegrees=45
  twistMotion=Free
 }
 {
  bone1Name=BASE_dyn_collar_l_01
  bone2Name=spine_05
  swing1LimitDegrees=45
  swing1Motion=Free
  swing2LimitDegrees=45
  swing2Motion=Free
  twistLimitDegrees=45
  twistMotion=Free
 }
 {bone1Name=dyn_big_collar_l_01, bone2Name=BASE_dyn_collar_l_01, swing1LimitDegrees=5, swing2LimitDegrees=10, twistLimitDegrees=5}
 {
  bone1Name=BASE_dyn_collar_r_01
  bone2Name=spine_05
  swing1LimitDegrees=45
  swing1Motion=Free
  swing2LimitDegrees=45
  swing2Motion=Free
  twistLimitDegrees=45
  twistMotion=Free
 }
 {bone1Name=dyn_big_collar_r_01, bone2Name=BASE_dyn_collar_r_01, swing1LimitDegrees=5, swing2LimitDegrees=10, twistLimitDegrees=5}
 {
  bone1Name=clavicle_l
  bone2Name=spine_05
  swing1LimitDegrees=45
  swing1Motion=Free
  swing2LimitDegrees=45
  swing2Motion=Free
  twistLimitDegrees=45
  twistMotion=Free
 }
 {
  bone1Name=upperarm_l
  bone2Name=clavicle_l
  swing1LimitDegrees=45
  swing1Motion=Free
  swing2LimitDegrees=45
  swing2Motion=Free
  twistLimitDegrees=45
  twistMotion=Free
 }
 {
  bone1Name=clavicle_r
  bone2Name=spine_05
  swing1LimitDegrees=45
  swing1Motion=Free
  swing2LimitDegrees=45
  swing2Motion=Free
  twistLimitDegrees=45
  twistMotion=Free
 }
 {
  bone1Name=upperarm_r
  bone2Name=clavicle_r
  swing1LimitDegrees=45
  swing1Motion=Free
  swing2LimitDegrees=45
  swing2Motion=Free
  twistLimitDegrees=45
  twistMotion=Free
 }
 {
  bone1Name=extra_spine_05
  bone2Name=spine_05
  swing1LimitDegrees=45
  swing1Motion=Free
  swing2LimitDegrees=45
  swing2Motion=Free
  twistLimitDegrees=45
  twistMotion=Free
 }
 {
  bone1Name=head
  bone2Name=neck_02
  swing1LimitDegrees=45
  swing1Motion=Free
  swing2LimitDegrees=45
  swing2Motion=Free
  twistLimitDegrees=45
  twistMotion=Free
 }
 {
  bone1Name=dyn_f_side_hair_BASE_l
  bone2Name=head
  swing1LimitDegrees=45
  swing1Motion=Free
  swing2LimitDegrees=45
  swing2Motion=Free
  twistLimitDegrees=45
  twistMotion=Free
 }
 {bone1Name=dyn_f_side_hair_01_l, bone2Name=dyn_f_side_hair_BASE_l, swing1LimitDegrees=10, swing2LimitDegrees=30, twistLimitDegrees=10}
 {bone1Name=dyn_f_side_hair_02_l, bone2Name=dyn_f_side_hair_01_l, swing1LimitDegrees=20, swing2LimitDegrees=20, twistLimitDegrees=30}
 {
  bone1Name=dyn_f_side_hair_BASE_r
  bone2Name=head
  swing1LimitDegrees=45
  swing1Motion=Free
  swing2LimitDegrees=45
  swing2Motion=Free
  twistLimitDegrees=45
  twistMotion=Free
 }
 {bone1Name=dyn_f_side_hair_01_r, bone2Name=dyn_f_side_hair_BASE_r, swing1LimitDegrees=20, swing2LimitDegrees=5, twistLimitDegrees=5}
 {bone1Name=dyn_f_side_hair_02_r, bone2Name=dyn_f_side_hair_01_r, swing1LimitDegrees=20, swing2LimitDegrees=10, twistLimitDegrees=5}
 {
  bone1Name=neck_02
  bone2Name=spine_05
  swing1LimitDegrees=45
  swing1Motion=Free
  swing2LimitDegrees=45
  swing2Motion=Free
  twistLimitDegrees=45
  twistMotion=Free
 }
 {
  bone1Name=dyn_back_hair_BASE_m
  bone2Name=head
  swing1LimitDegrees=45
  swing1Motion=Free
  swing2LimitDegrees=45
  swing2Motion=Free
  twistLimitDegrees=45
  twistMotion=Free
 }
 {bone1Name=dyn_back_hair_01_m, bone2Name=dyn_back_hair_BASE_m, swing1LimitDegrees=45, swing2LimitDegrees=20, twistLimitDegrees=5}
 {bone1Name=dyn_back_hair_02_m, bone2Name=dyn_back_hair_01_m, swing1LimitDegrees=45, swing2LimitDegrees=5, twistLimitDegrees=5}
 {bone1Name=dyn_back_hair_03_m, bone2Name=dyn_back_hair_02_m, swing1LimitDegrees=45, swing2LimitDegrees=5, twistLimitDegrees=5}
 {
  bone1Name=dyn_side_hair_BASE_l
  bone2Name=head
  swing1LimitDegrees=45
  swing1Motion=Free
  swing2LimitDegrees=45
  swing2Motion=Free
  twistLimitDegrees=45
  twistMotion=Free
 }
 {bone1Name=dyn_side_hair_01_l, bone2Name=dyn_side_hair_BASE_l, swing1LimitDegrees=45, swing2LimitDegrees=10, twistLimitDegrees=5}
 {bone1Name=dyn_side_hair_02_l, bone2Name=dyn_side_hair_01_l, swing1LimitDegrees=45, swing2LimitDegrees=5, twistLimitDegrees=5}
 {bone1Name=dyn_side_hair_03_l, bone2Name=dyn_side_hair_02_l, swing1LimitDegrees=45, swing2LimitDegrees=5, twistLimitDegrees=5}
 {
  bone1Name=dyn_side_hair_BASE_r
  bone2Name=head
  swing1LimitDegrees=45
  swing1Motion=Free
  swing2LimitDegrees=45
  swing2Motion=Free
  twistLimitDegrees=45
  twistMotion=Free
 }
 {bone1Name=dyn_side_hair_01_r, bone2Name=dyn_side_hair_BASE_r, swing1LimitDegrees=45, swing2LimitDegrees=10, twistLimitDegrees=20}
 {bone1Name=dyn_side_hair_02_r, bone2Name=dyn_side_hair_01_r, swing1LimitDegrees=45, swing2LimitDegrees=5, twistLimitDegrees=20}
 {bone1Name=dyn_side_hair_03_r, bone2Name=dyn_side_hair_02_r, swing1LimitDegrees=45, swing2LimitDegrees=5, twistLimitDegrees=20}
 {
  bone1Name=dyn_side_hair_03_l
  bone2Name=dyn_back_hair_03_m
  swing1LimitDegrees=45
  swing1Motion=Free
  swing2LimitDegrees=45
  swing2Motion=Free
  twistLimitDegrees=45
  twistMotion=Free
 }
 {
  bone1Name=dyn_side_hair_03_r
  bone2Name=dyn_back_hair_03_m
  swing1LimitDegrees=45
  swing1Motion=Free
  swing2LimitDegrees=45
  swing2Motion=Free
  twistLimitDegrees=45
  twistMotion=Free
 }
]
