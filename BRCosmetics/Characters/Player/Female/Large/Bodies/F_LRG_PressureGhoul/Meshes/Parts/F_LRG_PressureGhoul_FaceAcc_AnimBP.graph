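# Anim Blueprint for the PressureGhoul female large face accessory, adding secondary motion to the attached part.
# It blends parent and body-mesh poses by bone, then runs a Control Rig, two head-space RigidBody nodes (normal and DBNO-carried, mutually exclusive) and a post-sim Control Rig.
# Updates set gravity via a deprecated helper and switch velocity scaling for skydiving.
# With a front-end anim instance, initialization delays simulation and hair-position flags; debug prints remain.
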AnimBlueprint /BRCosmetics/Characters/Player/Female/Large/Bodies/F_LRG_PressureGhoul/Meshes/Parts/F_LRG_PressureGhoul_FaceAcc_AnimBP
target_skeleton=&/F_LRG_PressureGhoul_FaceAcc_Skeleton
graphs [
 AnimGraph {nodes [
   n1 = Layered blend per bone(
    BasePose=Copy Pose From Mesh() {bUseAttachedParent=true}
    BlendPoses_0=Copy Pose From Mesh(SourceMeshComponent=Get BodyMesh() -> (BodyMesh:Skeletal Mesh Component Object Reference)) {bUseAttachedParent=true}
   ) {layerSetup [{branchFilters [{boneName=upperarm_l}, {boneName=upperarm_r}, {boneName=dyn_big_collar_l_01}, {boneName=dyn_big_collar_r_01}]}]}
   n2 = Control Rig!(
    Source=@n1
    Alpha
    sky_diving:bool=Get bIsSkyDiving()
    crouching:bool=Get bIsCrouching()
    frontend:bool=Get bIsInFrontEnd()
    reload_dyn:bool=Get reload_dyn() -> (reload_dyn:bool)
    reload_hair_position:bool=Get reload_hair_position() -> (reload_hair_position:bool)
    DBNOCarried:bool=Get bIsDBNOCarried()
   ) -> (Pose) {
    alphaBoolBlend {blendOption=Linear}
    controlRigAssetReference {blueprintRigClass=&/F_LRG_PressureGhoul_FaceAcc_CtrlRig.F_LRG_PressureGhoul_FaceAcc_CtrlRig_C}
   }
   n3 = Get acc_vel_scale() -> (acc_vel_scale:Vector)
   n4 = RigidBody!(
    ComponentPose=Local To Component(LocalPose=@n2)
    bAlphaBoolEnabled:bool=Get rigid_body() -> (rigid_body:bool)
    OverrideWorldGravity:Vector=Get Gravity Override()
    ExternalForce
    ComponentLinearAccScale:Vector=@n3
    ComponentLinearVelScale:Vector=@n3
   ) -> (Pose) {
    alphaInputType=Bool
    bOverrideWorldGravity=true
    overridePhysicsAsset=&/F_LRG_PressureGhoul_FaceAcc_Physics_RigidBody
    simulationSpace=BaseBoneSpace
    alphaBoolBlend {blendInTime=0.2, blendOption=Linear, blendOutTime=0.2}
    baseBoneRef {boneName=head}
    componentAppliedLinearAccClamp=<2000 2000 2000>
    componentLinearAccScale=<0.13 0.24 0.11>
    componentLinearVelScale=<0.13 0.24 0.11>
    overrideWorldGravity {z=-1500}
    simSpaceSettings {maxAngularAcceleration=3000, maxAngularVelocity=3000, maxLinearAcceleration=3000, maxLinearVelocity=3000}
   }
   n5 = Get acc_vel_scale()
   n6 = RigidBody(
    ComponentPose=@n4
    bAlphaBoolEnabled=Get bIsDBNOCarried()
    OverrideWorldGravity=Get Gravity Override()
    ComponentLinearAccScale=@n5
    ComponentLinearVelScale=@n5
   ) {
    alphaInputType=Bool
    bOverrideWorldGravity=true
    overridePhysicsAsset=&/F_LRG_PressureGhoul_FaceAcc_Physics_RigidBody_DBNO
    simulationSpace=BaseBoneSpace
    alphaBoolBlend {blendInTime=0.2, blendOption=Linear, blendOutTime=0.2}
    baseBoneRef {boneName=head}
    componentAppliedLinearAccClamp=<2000 2000 2000>
    componentLinearAccScale=<0.13 0.24 0.11>
    componentLinearVelScale=<0.13 0.24 0.11>
    overrideWorldGravity {z=-1500}
    simSpaceSettings {maxAngularAcceleration=3000, maxAngularVelocity=3000, maxLinearAcceleration=3000, maxLinearVelocity=3000}
   }
   n7 = Control Rig!(Source=Component To Local(ComponentPose=@n6), Alpha) -> (Pose) {
    alphaBoolBlend {blendOption=Linear}
    controlRigAssetReference {blueprintRigClass=&/F_LRG_PressureGhoul_FaceAcc_CtrlRig_postsim.F_LRG_PressureGhoul_FaceAcc_CtrlRig_postsim_C}
   }
   Output Pose(Result=@n7) {layerGroup=None}
  ]}
 EventGraph {nodes [
   Event BlueprintUpdateAnimation()
   n1 = Get bIsDBNOCarried()
   > Set rigid_body(rigid_body:bool=AND Boolean(A=Get reload_dyn(), B=NOT Boolean(A=@n1))) -> (Output_Get:bool)
   > n2 = Gravity Override State Function DEPRECATED(CharacterPartAnimInstance=Self-Reference(), GameTimeInSeconds=GetGameTimeInSeconds(), GravityOverrideParams=Get Gravity Override Params())
   > Set Gravity Override(Gravity Override=@n2)
   > n3 = Branch(Condition=Get bIsSkyDiving())
   > Set acc_vel_scale(acc_vel_scale:Vector="0.13,0.24,0.11") -> (Output_Get:Vector)
   Set acc_vel_scale(execute=@n3.else, acc_vel_scale="0.35,0.35,0.35")
   Event BlueprintInitializeAnimation()
   > n4 = Is Valid(InputObject=GetOwnerFrontEndAnimInstance())
   Set reload_dyn(execute=@n4.Is Not Valid, reload_dyn:bool=true) -> (Output_Get:bool)
   > n5 = Set reload_hair_position(reload_hair_position:bool=true) -> (Output_Get:bool)
   > PrintString!(execute, InString="To String (Boolean)"(InBool=@n5), bPrintToScreen, bPrintToLog, TextColor, Duration, Key) -> (then)
   PrintString(execute=@n4.Is Valid, InString=HI)
   > Set reload_hair_position(reload_hair_position=false)
   > Set reload_dyn(reload_dyn=false)
   > n6 = Sequence()
   > RetriggerableDelay(Duration=0.4)
   > Set reload_dyn(reload_dyn=true)
   RetriggerableDelay(execute=@n6.then_1, Duration=0.7)
   > Set reload_hair_position(reload_hair_position=true)
   n7 = Event OnCharacterCustomizationLoaded()
   > Set BodyMesh(BodyMesh="Get (a copy)"(Array:Array of Skeletal Mesh Component Budgeted Object References=@n7.CustomMeshParts, Dimension 1=1) -> (Output:Skeletal Mesh Component Budgeted Object Reference))
  ]}
]
variables [
 acc_vel_scale {type=Vector}
 BodyMesh {type=SkeletalMeshComponent}
 Gravity Override {type=Vector}
 Gravity Override Params {
  type=GravityOverrideParamsStruct
  default {
   skydive_FloatingDown {gravityOverride=<0 -700 100>, windAmplitude=<200 -500 200>, windOffset=<0 -600 0>}
   skydive_GoingUp {gravityOverride=<0 -300 100>}
   tactical_Sprint {gravityOverride=<0 -450 -1300>}
  }
 }
 reload_dyn {type=boolean}
 reload_hair_position {type=boolean}
 rigid_body {type=boolean}
]
